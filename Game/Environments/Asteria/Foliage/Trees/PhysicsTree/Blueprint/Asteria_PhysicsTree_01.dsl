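; Blueprint for an Asteria physics tree that can be felled, handling its felling effects, analytics and material variation.
; The construction script, and BeginPlay where cosmetic events can run, write base, trunk and leaf array-texture indices to the mesh's custom primitive data.
; On felling, the server reports analytics through the spawned physics log and sends a fallen-tree stat event.
; The replicated fall vector spawns a crown VFX actor rotated toward the horizontal fall direction.
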(event EventBeginPlay
  (Networking|CanExecuteCosmeticEvents
    (:True
      (Default|SetArrayTextureCustomPrimData))))

(event EventObjectFelledonServer
  ;[1 1 1] Object felled
  (bind _asbga_physics_tree_log (Utilities|Casting|CastToBGA_PhysicsTreeLog (Variables|FellingResults|GetSpawnedPhysicsObject)))
  (Class|BGAPhysicsTreeLog|ReportTreeFellingtoAnalytics _asbga_physics_tree_log (Variables|FellingResults|GetFelledDamageCauser))
  (CallFunction|SendFallenTreeStatEvent (Variables|FellingResults|GetFelledDamageCauser)) ; CustomStatEvent for Weekly W05Q04
)

(event EventFellingVectorReplicated
  (bind _fallvector_fx (Variables|FellingResults|GetFallVector_FX))
  (bind _x (.x _fallvector_fx))
  (bind _y (.y _fallvector_fx))
  (bind _returnvalue (Math|Vector|MakeVector _x _y))
  (bind _returnvalue_2 (Math|Vector|Normalize _returnvalue))
  ;[1 1 1] Custom VFX for felling
  (bind _returnvalue_3 (Game|SpawnActorfromClass "/Game/Environments/Asteria/Foliage/Trees/PhysicsTree/Blueprint/BP_Tree_Crown.BP_Tree_Crown_C" (Transformation|GetWorldTransform (Variables|Debug|GetStaticMeshComponent)) "Undefined" "OverrideRootScale" 0 _returnvalue_2 0))
  (Variables|/Game/Environments/Asteria/Foliage/Trees/PhysicsTree/Blueprint/BP_Tree_Crown.BP_Tree_Crown_C|SetAxisRotationAngle (+ (Math|Trig|Atan2(Degrees) (.x _returnvalue_2) (.y _returnvalue_2)) 9e+01) _returnvalue_3)
  (Variables|/Game/Environments/Asteria/Foliage/Trees/PhysicsTree/Blueprint/BP_Tree_Crown.BP_Tree_Crown_C|SetFallVector _returnvalue_2 _returnvalue_3))

(fn SetArrayTextureCustomPrimData
  (Rendering|Material|SetCustomPrimitiveDataFloat (Variables|Debug|GetStaticMeshComponent) 17 (Math|Conversions|ToFloat(Integer) (Variables|Default|GetArrayIndex)))
  (Rendering|Material|SetCustomPrimitiveDataFloat (Variables|Debug|GetStaticMeshComponent) 18 (Math|Conversions|ToFloat(Integer) (Variables|Default|GetArrayIndex_Trunk)))
  (Rendering|Material|SetCustomPrimitiveDataFloat (Variables|Debug|GetStaticMeshComponent) 19 (Math|Conversions|ToFloat(Integer) (Variables|Default|GetArrayIndex_Leaves))))

(fn ConstructionScript
  (Default|SetArrayTextureCustomPrimData))
